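# Material function for Niagara effects materials, marked reference only, showing common UV flip and rotation techniques as eight outputs.
# A UV input, defaulting to TextureCoordinate, passes through as Standard UV and feeds seven variants built by splitting, swapping or inverting components with OneMinus.
# Rotate 90 UV feeds G into both X and Y, collapsing to a diagonal, unlike the other swapped variants.
# Rotate 180 UV inverts only X.
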MaterialFunction /Game/Effects/Niagara/MasterMaterials/Functions/Mojo_MF/MF_FX_Common_UV_Flip_And_Rotate_Ref
graph {
 properties {description="Common UV flip and rotation techniques (Reference Only)"}
 nodes [
  n1 = FunctionInput(Preview=TextureCoordinate()) {bUsePreviewValueAsDefault=true, inputName=In TexCoords, inputType=FunctionInput_Vector2}
  FunctionOutput(None=@n1) {outputName=Standard UV}
  n2 = MaterialFunctionCall(Float2=@n1) {materialFunction=&BreakOutFloat2Components}
  FunctionOutput(None=MaterialFunctionCall(X=@n2.G, Y=@n2.G) {materialFunction=&MakeFloat2}.Result) {outputName=Rotate 90 UV, sortPriority=1}
  n3 = MaterialFunctionCall(Float2=@n1) {materialFunction=&BreakOutFloat2Components}
  FunctionOutput(None=MaterialFunctionCall(X=OneMinus(None=@n3.R), Y=@n3.G) {materialFunction=&MakeFloat2}.Result) {outputName=Rotate 180 UV, sortPriority=2}
  n4 = MaterialFunctionCall(Float2=@n1) {materialFunction=&BreakOutFloat2Components}
  FunctionOutput(None=MaterialFunctionCall(X=@n4.R, Y=OneMinus(None=@n4.G)) {materialFunction=&MakeFloat2}.Result) {outputName=Standard Flip UV, sortPriority=4}
  n5 = MaterialFunctionCall(Float2=@n1) {materialFunction=&BreakOutFloat2Components}
  FunctionOutput(None=MaterialFunctionCall(X=@n5.G, Y=OneMinus(None=@n5.R)) {materialFunction=&MakeFloat2}.Result) {outputName=Rotate 90 Flip UV, sortPriority=5}
  n6 = MaterialFunctionCall(Float2=@n1) {materialFunction=&BreakOutFloat2Components}
  FunctionOutput(None=MaterialFunctionCall(X=OneMinus(None=@n6.G), Y=@n6.R) {materialFunction=&MakeFloat2}.Result) {outputName=Rotate 270 UV, sortPriority=3}
  n7 = MaterialFunctionCall(Float2=@n1) {materialFunction=&BreakOutFloat2Components}
  FunctionOutput(None=MaterialFunctionCall(X=OneMinus(None=@n7.R), Y=OneMinus(None=@n7.G)) {materialFunction=&MakeFloat2}.Result) {outputName=Rotate 180 Flip UV, sortPriority=6}
  n8 = MaterialFunctionCall(Float2=@n1) {materialFunction=&BreakOutFloat2Components}
  FunctionOutput(None=MaterialFunctionCall(X=OneMinus(None=@n8.G), Y=OneMinus(None=@n8.R)) {materialFunction=&MakeFloat2}.Result) {outputName=Rotate 270 Flip UV, sortPriority=7}
 ]
}
