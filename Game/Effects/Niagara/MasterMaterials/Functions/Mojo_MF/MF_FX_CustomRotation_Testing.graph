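# Experimental material function that rotates UVs around a pivot for FX master materials.
# It offsets the UVs by (0.5, 0.5) minus the pivot input, applies a 2D rotation using Sine and Cosine nodes, then restores the offset.
# The angle input defaults to scaled Time, giving a continuous spin.
# Custom HLSL sin and cos nodes also read the angle but do not feed the output.
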MaterialFunction /Game/Effects/Niagara/MasterMaterials/Functions/Mojo_MF/MF_FX_CustomRotation_Testing
graph {
 properties {description="Custom UV rotation I have been toying with. (probably should use custom rotator if you want)"}
 nodes [
  n1 = Time()
  MaterialExpressionReroute_1 = Reroute(None=FunctionInput(Preview=Multiply(A=@n1) {constB=0.2}) {bUsePreviewValueAsDefault=true, description=angle, inputName=a, inputType=FunctionInput_Scalar, sortPriority=2}) {desc="a aka angle "}
  Custom!(x=@MaterialExpressionReroute_1) {
   code="return sin(x);"
   description="Sin(x)"
   outputType=CMOT_Float1
   inputs [{
     inputName=x
     input {expression=&self:MaterialExpressionReroute_1}
    }]
  }
  Custom(x=@MaterialExpressionReroute_1) {
   code="return cos(x);"
   description="Cos(x)"
   outputType=CMOT_Float1
   inputs [{
     inputName=x
     input {expression=&self:MaterialExpressionReroute_1}
    }]
  }
  n2 = FunctionInput(Preview=Constant2Vector()) {bUsePreviewValueAsDefault=true, description=pivot point, inputName=pivot, inputType=FunctionInput_Vector2, sortPriority=1}
  n3 = Subtract(A=Constant2Vector() {g=0.5, r=0.5}, B=@n2)
  n4 = Subtract(A=FunctionInput(Preview=TextureCoordinate()) {bUsePreviewValueAsDefault=true, inputName=st, inputType=FunctionInput_Vector2}, B=@n3)
  n5 = ComponentMask(None=@n4) {r=true}
  n6 = Cosine(None=@MaterialExpressionReroute_1) {period=6.28}
  n7 = ComponentMask(None=@n4) {g=true}
  n8 = Sine(None=@MaterialExpressionReroute_1) {period=6.28}
  n9 = Subtract(A=Multiply(A=@n5, B=@n6), B=Multiply(A=@n7, B=@n8))
  n10 = Add(A=Multiply(A=@n5, B=@n8), B=Multiply(A=@n7, B=@n6))
  FunctionOutput(None=Add(A=MaterialFunctionCall(X=@n9, Y=@n10) {materialFunction=&MakeFloat2}.Result, B=@n3)) {outputName=Rotated UVs}
 ]
}
