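# Material function that fakes a 3D perspective wobble on UVs for FX materials.
# It centers the input UVs, appends a slowly oscillating sine depth, then rotates that 3D point around two axes by time-driven cosine and sine angles.
# "Out UVs V2" divides rotated XY by depth for perspective, while "Out UVs" only doubles it.
# All motion is hardcoded, and neither output re-adds the centering offset.
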MaterialFunction /Game/Effects/Niagara/MasterMaterials/Functions/Mojo_MF/MF_FX_FakePerspective_UV
graph {nodes [
  n1 = Multiply(A=Cosine(None=Time()) {period=6.28}, B=Constant() {r=0.5})
  n2 = TextureCoordinate()
  n3 = Multiply(A=Time()) {constB=0.1}
  n4 = AppendVector(
   A=Add(A=FunctionInput(Preview=@n2) {inputName=In UVs, inputType=FunctionInput_Vector2}, B=Constant2Vector() {g=-0.5, r=-0.5})
   B=Add(A=Constant() {r=1}, B=Sine(None=@n3) {period=6.28})
  )
  n5 = ComponentMask(None=@n4) {r=true}
  n6 = Sine(None=@n1) {period=6.28}
  n7 = Constant() {r=-1}
  n8 = ComponentMask(None=@n4) {b=true}
  n9 = Add(A=Multiply(A=Cosine(None=@n1) {period=6.28}, B=@n5), B=Multiply(A=Multiply(A=@n6, B=@n7), B=@n8))
  n10 = Add(A=Multiply(A=Sine(None=@n1) {period=6.28}, B=@n5), B=Multiply(A=Cosine(None=@n1) {period=6.28}, B=@n8))
  n11 = Multiply(A=Sine(None=Time()) {period=6.28}, B=Constant() {r=0.5})
  n12 = Cosine(None=@n11) {period=6.28}
  n13 = ComponentMask(None=@n4) {g=true}
  n14 = Multiply(A=Sine(None=@n11) {period=6.28}, B=Constant() {r=-1})
  n15 = Sine(None=@n11) {period=6.28}
  n16 = Cosine(None=@n11) {period=6.28}
  n17 = AppendVector(A=Add(A=Multiply(A=@n13, B=@n12), B=Multiply(A=@n10, B=@n14)), B=Add(A=Multiply(A=@n13, B=@n15), B=Multiply(A=@n10, B=@n16)))
  n18 = AppendVector(A=ComponentMask(None=AppendVector(A=@n9, B=@n10)) {r=true}, B=ComponentMask(None=@n17) {r=true})
  FunctionOutput(None=Divide(A=@n18, B=ComponentMask(None=@n17) {g=true})) {outputName=Out UVs V2, sortPriority=1}
  FunctionOutput(None=Multiply(A=@n18) {constB=2, desc=optional can remove if needed}) {outputName=Out UVs}
 ]}
