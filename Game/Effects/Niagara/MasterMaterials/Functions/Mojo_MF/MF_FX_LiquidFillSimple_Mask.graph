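# Material function that fakes a simple liquid fill on a mesh, outputting color, opacity and a foam mask.
# World-space height relative to the object position is stepped against Fill Amount for opacity, and a second step offset by Foam Amount isolates the foam band below the fill line.
# TwoSidedSign gives back faces Color 1, posing as the liquid's top surface, while front faces blend Color 2 to Color 3 by the foam mask.
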MaterialFunction /Game/Effects/Niagara/MasterMaterials/Functions/Mojo_MF/MF_FX_LiquidFillSimple_Mask
graph {nodes [
  n1 = ComponentMask(None=Subtract(A=WorldPosition().XYZ, B=ObjectPositionWS())) {b=true}
  n2 = FunctionInput(Preview=Constant() {r=1}) {bUsePreviewValueAsDefault=true, inputName=Fill Amount, inputType=FunctionInput_Scalar, sortPriority=1}
  n3 = Subtract(A=@n2, B=FunctionInput(Preview=Constant() {r=10}) {bUsePreviewValueAsDefault=true, inputName=Foam Amount, inputType=FunctionInput_Scalar})
  n4 = Step(Y=@n1, X=@n2)
  n5 = Saturate(None=Subtract(A=@n4, B=Step(Y=@n1, X=@n3)))
  n6 = FunctionInput(Preview=Constant3Vector() {constant=<rgba 0.3046 1 0 0>}) {bUsePreviewValueAsDefault=true, inputName=Color 2, sortPriority=3}
  n7 = FunctionInput(Preview=Constant3Vector() {constant=<rgba 0 1 0.9916 0>}) {bUsePreviewValueAsDefault=true, inputName=Color 3, sortPriority=4}
  n8 = TwoSidedSign()
  n9 = Constant3Vector() {constant=<rgba 1 1 1 0>}
  FunctionOutput(None=LinearInterpolate(
    A=FunctionInput(Preview=@n9) {bUsePreviewValueAsDefault=true, inputName=Color 1, sortPriority=2}
    B=LinearInterpolate(A=@n6, B=@n7, Alpha=@n5)
    Alpha=Saturate(None=@n8)
   )) {outputName=Out Color}
  FunctionOutput(None=Saturate(None=@n4)) {outputName=Out Opacity, sortPriority=1}
  FunctionOutput(None=@n5) {outputName=Out Foam Mask, sortPriority=2}
 ]}
