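# Testing material function that builds an animated grid of UV cells for FX, outputting per-cell UVs plus random values per cell, row and column.
# The input UV is globally rotated and offset, tiled through MF_UI_Pattern, then each cell is scaled, offset and rotated.
# Floored UVs feed MF_UI_Random for the random outputs.
# Default cell offset and rotation are time-driven through Speed and Time inputs, so cells move when those are left unconnected.
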MaterialFunction /Game/Effects/Niagara/MasterMaterials/Functions/Mojo_MF/MF_FX_ParticleGrid_Testing
graph {nodes [
  Frac(None=Time())
  n1 = MaterialFunctionCall() {materialFunction=&GetUserInterfaceUV}
  n2 = FunctionInput(Preview=Constant()) {
   bUsePreviewValueAsDefault=true
   description=Define a rotation for the entire grid
   inputName=Global Rotation
   inputType=FunctionInput_Scalar
   sortPriority=2
  }
  n3 = NamedRerouteDeclaration(None=MaterialFunctionCall(
    UVs=FunctionInput(Preview=@n1.Normalized UV) {bUsePreviewValueAsDefault=true, description=Set the UV, inputName=UV, inputType=FunctionInput_Vector2}
    "Rotation Angle (0-1)"=@n2
   ) {materialFunction=&CustomRotator}.Rotated Values) {name=Global Rotated UV, nodeColor=<rgba 0 0 0 1>}
  n4 = FunctionInput(Preview=Constant()) {
   bUsePreviewValueAsDefault=true
   description=Define an offset for the entire grid
   inputName=Global Offset
   inputType=FunctionInput_Vector2
   sortPriority=1
  }
  n5 = Frac(None=@n4)
  n6 = NamedRerouteUsage()
  n7 = MaterialFunctionCall!(UV=Frac(None=Add(A=@n3, B=@n5) {constB=0}), PatternX=ComponentMask(None=@n6) {r=true}, PatternY=ComponentMask(None=@n6) {g=true}) {materialFunction=/FortUILibrary/MaterialFunctions/Basics/MF_UI_Pattern}
  n8 = FunctionInput(Preview=Constant() {r=0.5}) {bUsePreviewValueAsDefault=true, description=Set the individual scale, inputName=Cell Scale, inputType=FunctionInput_Scalar, sortPriority=11}
  n9 = MaterialFunctionCall(UVs=@n7.PatternUVs, Texture Scale=@n8) {materialFunction=&ScaleUVsByCenter}
  n10 = NamedRerouteDeclaration(None=FunctionInput(Preview=Constant() {r=0.1}) {bUsePreviewValueAsDefault=true, inputName=Speed, inputType=FunctionInput_Scalar}) {name=Speed, nodeColor=<rgba 1 0.1641 0 1>}
  n11 = NamedRerouteDeclaration(None=FunctionInput(Preview=MaterialFunctionCall(Speed=@n10) {materialFunction=&DebugTimeSine}.Result) {bUsePreviewValueAsDefault=true, inputName=Time, inputType=FunctionInput_Scalar}) {name=timeInput, nodeColor=<rgba 0 0.2266 1 1>}
  n12 = FunctionInput(Preview=Constant()) {bUsePreviewValueAsDefault=true, inputName=Time Add, inputType=FunctionInput_Scalar}
  n13 = Multiply(
   A=Add(A=Multiply(A=@n10, B=@n11), B=NamedRerouteDeclaration(None=@n12) {name=Timeoffset, nodeColor=<rgba 0 0.1562 1 1>})
   B=NamedRerouteUsage()
  )
  n14 = FunctionInput(Preview=AppendVector(A=LinearInterpolate(Alpha=@n13) {constA=-0.5, constB=0.5, desc=Oscillation}, B=Constant())) {
   bUsePreviewValueAsDefault=true
   description=Set the individual offset
   inputName=Cell Offset
   inputType=FunctionInput_Vector2
   sortPriority=10
  }
  n15 = Add(A=Multiply(A=NamedRerouteUsage(), B=NamedRerouteUsage()) {constB=0.1}, B=NamedRerouteUsage())
  n16 = FunctionInput(Preview=Add(A=NamedRerouteUsage(), B=Frac(None=@n15))) {
   bUsePreviewValueAsDefault=true
   description=Set the individual rotation
   inputName=Cell Rotation
   inputType=FunctionInput_Scalar
   sortPriority=12
  }
  FunctionOutput(None=MaterialFunctionCall(UVs=Subtract(A=@n9.UVs, B=@n14) {constB=0}, "Rotation Angle (0-1)"=@n16) {materialFunction=&CustomRotator}.Rotated Values) {description=Returns the UV for all the cells, outputName=Cells UV}
  n17 = Floor(None=NamedRerouteUsage())
  n18 = AppendVector(A=ComponentMask(None=@n17) {r=true}, B=ComponentMask(None=@n17) {g=true})
  n19 = FunctionInput(Preview=Constant()) {
   bUsePreviewValueAsDefault=true
   description=Set the seed to randomize differently the values
   inputName=Random Seed
   inputType=FunctionInput_Scalar
   sortPriority=32
  }
  FunctionOutput(None=NamedRerouteDeclaration(None=MaterialFunctionCall!(UV=@n18, Seed=@n19) {materialFunction=&MF_UI_Random}.Result) {name=Random Cell Values, nodeColor=<rgba 0 0 0 1>}) {description=Output random values for all the cells, outputName=Random Cell Values, sortPriority=1}
  n20 = FunctionInput(Preview=Constant() {r=4}) {bUsePreviewValueAsDefault=true, description=Define the tiling value, inputName=Tiling, inputType=FunctionInput_Vector2, sortPriority=3}
  n21 = Floor(None=Multiply(A=NamedRerouteUsage(), B=NamedRerouteDeclaration(None=@n20) {name=Tiling Value, nodeColor=<rgba 0 0 0 1>}))
  n22 = ComponentMask(None=@n21) {g=true}
  n23 = Constant() {r=0.3}
  FunctionOutput(None=NamedRerouteDeclaration(None=MaterialFunctionCall(UV=@n22, Seed=@n23) {materialFunction=&MF_UI_Random}.Result) {name=Random Rows, nodeColor=<rgba 0 0 0 1>}) {description=Output randomized rows, outputName=Random Rows, sortPriority=10}
  n24 = ComponentMask(None=@n21) {r=true}
  FunctionOutput(None=NamedRerouteDeclaration(None=MaterialFunctionCall(UV=@n24) {materialFunction=&MF_UI_Random}.Result) {name=Random Columns, nodeColor=<rgba 0 0 0 1>}) {description=Output randomized columns, outputName=Random Columns, sortPriority=11}
  NamedRerouteDeclaration(None=@n7.UVs) {name=Single UV, nodeColor=<rgba 0 0 0 1>}
  Time()
 ]}
