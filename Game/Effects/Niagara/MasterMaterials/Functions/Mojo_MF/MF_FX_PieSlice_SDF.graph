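# Material function that computes a 2D pie slice signed distance field for FX materials, from texture coordinates, an aperture angle and a radius.
# It centres the UVs, mirrors X, and turns the angle into a sine/cosine edge direction.
# The max of circle distance and signed edge distance feeds the raw SDF output and, through MF_FX_Fill, a solid output.
# When the angle input is unconnected, a time-driven cosine animates the slice opening and closing.
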MaterialFunction /Game/Effects/Niagara/MasterMaterials/Functions/Mojo_MF/MF_FX_PieSlice_SDF
graph {nodes [
  Constant() {r=3}
  n1 = Add(
   A=FunctionInput(Preview=TextureCoordinate()) {bUsePreviewValueAsDefault=true, description=texcoords, inputName=p, inputType=FunctionInput_Vector2}
   B=Constant2Vector() {g=-0.5, r=-0.5}
  )
  n2 = AppendVector(A=Abs(None=ComponentMask(None=@n1) {r=true}), B=ComponentMask(None=@n1) {g=true})
  n3 = Time()
  n4 = Constant() {r=0.52}
  n5 = Multiply(A=Constant() {r=0.5}, B=Cosine(None=Multiply(A=@n3, B=@n4)) {desc=period 3.14, period=3.14})
  n6 = Constant() {r=0.5}
  n7 = FunctionInput(Preview=Multiply(A=Constant() {r=3.14}, B=Add(A=@n5, B=@n6))) {bUsePreviewValueAsDefault=true, description=texcoords, inputName=c, inputType=FunctionInput_Vector2, sortPriority=1}
  n8 = AppendVector(A=Sine(None=@n7) {period=6.28}, B=Cosine(None=@n7) {period=6.18})
  n9 = Subtract(
   A=Multiply(A=ComponentMask(None=@n2) {r=true}, B=ComponentMask(None=@n8) {g=true})
   B=Multiply(A=ComponentMask(None=@n8) {r=true}, B=ComponentMask(None=@n2) {g=true})
  )
  n10 = DotProduct(A=@n2, B=@n8)
  n11 = Constant()
  n12 = FunctionInput(Preview=Constant() {r=0.25}) {bUsePreviewValueAsDefault=true, description=radius, inputName=r, sortPriority=2}
  n13 = Subtract(A=@n2, B=Multiply(A=@n8, B=Clamp(None=@n10, Min=@n11, Max=@n12)))
  n14 = Constant()
  n15 = Constant()
  n16 = Max(A=Subtract(A=Distance(A=@n2, B=@n15), B=@n12), B=Multiply(A=Distance(A=@n13, B=@n14), B=Sign(None=@n9)))
  FunctionOutput(None=MaterialFunctionCall!(x=@n16, size=Constant()) {materialFunction=&/MF_FX_Fill}.Result) {outputName=Result Solid, sortPriority=1}
  FunctionOutput(None=@n16) {outputName=Result SDF}
  OneMinus(None=@n16)
 ]}
