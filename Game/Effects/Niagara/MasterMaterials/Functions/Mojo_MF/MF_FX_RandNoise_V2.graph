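# Material function for FX materials that outputs pseudo-random noise from a 2D coordinate.
# It uses the classic sine hash: a dot product of the coordinate and a constant vector feeds a sine, which is scaled by a large multiplier and passed through Frac.
# Coordinate, dot vector and multiplier are all overridable Vector2 inputs, defaulting to the texture coordinate and built-in hash constants.
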MaterialFunction /Game/Effects/Niagara/MasterMaterials/Functions/Mojo_MF/MF_FX_RandNoise_V2
graph {nodes [
  n1 = TextureCoordinate()
  n2 = Constant2Vector() {g=78.233, r=12.9898}
  n3 = Sine(None=DotProduct(
    A=FunctionInput(Preview=@n1) {bUsePreviewValueAsDefault=true, inputName=In TexCoord, inputType=FunctionInput_Vector2}
    B=FunctionInput(Preview=@n2) {bUsePreviewValueAsDefault=true, inputName=Dot Product Value, inputType=FunctionInput_Vector2, sortPriority=1}
   ))
  n4 = FunctionInput(Preview=Constant() {r=443758.5312}) {bUsePreviewValueAsDefault=true, inputName=Multiply Value, inputType=FunctionInput_Vector2, sortPriority=2}
  FunctionOutput(None=Frac(None=Multiply(A=@n3, B=@n4)))
 ]}
