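# MF_FX_Ray_SDF is an effects material function built to output a repeating radial ray gradient from TexCoord and a count N.
# It centers the UVs, takes Arctangent2, divides by two pi, multiplies by N and keeps the fraction.
# As wired, both arctangent inputs read the R component, so the angle follows only that component's sign instead of sweeping around the center.
# A loose pi constant is unused.
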MaterialFunction /Game/Effects/Niagara/MasterMaterials/Functions/Mojo_MF/MF_FX_Ray_SDF
graph {nodes [
  Constant() {r=3.1416}
  n1 = TextureCoordinate()
  n2 = MaterialFunctionCall(Float2=Subtract(A=FunctionInput(Preview=@n1) {bUsePreviewValueAsDefault=true, inputName=TexCoord, inputType=FunctionInput_Vector2}, B=Constant() {r=0.5})) {materialFunction=&BreakOutFloat2Components}
  n3 = Divide(
   A=Arctangent2!(Y=@n2.R, X=@n2.R)
   B=MaterialFunctionCall(Multiplier=Constant() {r=2}) {materialFunction=&Pi}.Result
  )
  n4 = FunctionInput(Preview=Constant() {r=8}) {bUsePreviewValueAsDefault=true, inputName=N, inputType=FunctionInput_Scalar}
  FunctionOutput(None=Frac(None=Multiply(A=@n3, B=@n4)))
 ]}
