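# Material function for Niagara FX master materials that picks one of three inputs, values or textures, from a scalar selector described as a random 0–2 value.
# Nested If nodes compare the selector against rising thresholds, yielding the first input in the lowest band, the second in the middle band and the third otherwise.
# The last threshold test is redundant, as both of its connected branches return the third input.
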MaterialFunction /Game/Effects/Niagara/MasterMaterials/Functions/Mojo_MF/MF_FX_Select_3Way
graph {nodes [
  n1 = FunctionInput(Preview=Constant()) {
   bUsePreviewValueAsDefault=true
   description="Random 0 - 2 value goes into here. (randomized?)"
   inputName=Selector
   inputType=FunctionInput_Scalar
  }
  n2 = FunctionInput() {
   bUsePreviewValueAsDefault=true
   description="First input (can be a value or texture)"
   inputName="< 0.5"
   sortPriority=1
   previewValue=<0 0 1 1>
  }
  n3 = Constant() {r=1}
  n4 = FunctionInput() {
   bUsePreviewValueAsDefault=true
   description="Second input (can be a value or texture)"
   inputName="< 1"
   sortPriority=2
   previewValue=<1 0 0 1>
  }
  n5 = FunctionInput() {
   bUsePreviewValueAsDefault=true
   description="Third input (can be a value or texture)"
   inputName="< 1.5"
   sortPriority=3
   previewValue=<0 1 0 1>
  }
  n6 = If(A=Constant() {r=1.5}, B=@n1, "A > B"=@n5, "A < B"=@n5)
  FunctionOutput(None=If(A=Constant() {r=0.5}, B=@n1, "A > B"=@n2, "A < B"=If(A=@n3, B=@n1, "A > B"=@n4, "A < B"=@n6))) {description=outputs one of the four inputs., outputName=Selector Output}
 ]}
