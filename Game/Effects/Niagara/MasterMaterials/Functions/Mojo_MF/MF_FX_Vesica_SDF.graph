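# Material function that returns a vesica (lens) shape distance field for procedural FX materials.
# It offsets the st coordinates, which default to texture coordinates, horizontally by plus and minus half of the w input.
# Each offset set feeds MF_FX_Circle_SDF, and a Max of the two circle fields forms the overlap shape, with w setting the spacing between circle centers.
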MaterialFunction /Game/Effects/Niagara/MasterMaterials/Functions/Mojo_MF/MF_FX_Vesica_SDF
graph {nodes [
  n1 = FunctionInput(Preview=TextureCoordinate()) {bUsePreviewValueAsDefault=true, inputName=st, inputType=FunctionInput_Vector2}
  n2 = Constant() {r=0.2}
  n3 = AppendVector(
   A=Multiply(A=FunctionInput(Preview=@n2) {bUsePreviewValueAsDefault=true, inputName=w, inputType=FunctionInput_Scalar}, B=Constant() {r=0.5})
   B=Constant()
  )
  n4 = Subtract(A=@n1, B=@n3)
  n5 = Add(A=@n1, B=@n3)
  FunctionOutput(None=Max(
    A=MaterialFunctionCall!(In Texcoords=@n4) {materialFunction=&/MF_FX_Circle_SDF}.Result
    B=MaterialFunctionCall(In Texcoords=@n5) {materialFunction=&MF_FX_Circle_SDF}.Result
   ))
 ]}
