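# Material function for effects materials that outputs an erosion (dissolve) mask and an erosion edge mask.
# A selectable texture channel, defaulting to tiled noise, feeds a SmoothStep set by amount and contrast; the invert switch multiplies it by its inverse.
# The result optionally combines with a base mask channel and blends by influence, with a switch outputting 1 when disabled.
# Ceil operations isolate the lowest nonzero band as the edge.
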MaterialFunction /Game/Effects/Niagara/MasterMaterials/Functions/Mojo_MF/MF_Erosion_Complex_Base
graph {
 properties {bExposeToLibrary=true, description=Erosion Function, previewBlendMode=BLEND_Translucent}
 nodes [
  n1 = FunctionInput(Preview=Constant() {r=0.5}) {bUsePreviewValueAsDefault=true, inputName=Erosion Amount, inputType=FunctionInput_Scalar, sortPriority=2}
  n2 = FunctionInput(Preview=Constant() {r=0.8}) {bUsePreviewValueAsDefault=true, inputName=Erosion Contrast, inputType=FunctionInput_Scalar, sortPriority=2, previewValue=<0.5 0.5 0.5 0.5>}
  n3 = TextureCoordinate()
  n4 = Constant2Vector() {g=0.25, r=0.5}
  n5 = FunctionInput(Preview=TextureSample(UVs=Multiply(A=@n3, B=@n4)) {samplerType=SAMPLERTYPE_LinearColor, texture=&HighContrastNoise}.RGBA) {bUsePreviewValueAsDefault=true, inputName="Erosion Texture (RGBA)", inputType=FunctionInput_Vector4}
  n6 = ChannelMaskParameter(None=@n5) {group=Erotion, parameterName=Erosion Channel, sortPriority=-2}
  n7 = OneMinus(None=@n6.R)
  n8 = MaterialFunctionCall(
   Alpha=StaticSwitchParameter(True=Multiply(A=@n7, B=@n6.R), False=@n6.R) {group=Erotion, parameterName=Invert Erosion Values}
   Min=@n1
   Max=Add(A=@n1, B=OneMinus(None=@n2))
  ) {materialFunction=&SmoothStep}
  n9 = Ceil(None=@n8.Result)
  n10 = Constant() {r=0.1}
  n11 = OneMinus(None=Ceil(None=Subtract(A=@n8.Result, B=@n10)))
  FunctionOutput(None=Saturate(None=Multiply(A=@n9, B=@n11))) {outputName=Erosion Edge Mask, sortPriority=1}
  n12 = ChannelMaskParameter(None=@n5) {group=Erotion, maskChannel=Alpha, parameterName=Base Mask Channel, sortPriority=4, defaultValue=<rgba 0 0 0 1>}
  n13 = LinearInterpolate(
   A=StaticSwitchParameter(True=@n12.R, False=Constant() {r=1}) {group=Erotion, parameterName=Use Base Mask, sortPriority=3}
   B=StaticSwitchParameter(True=Multiply(A=@n12.R, B=@n8.Result), False=@n8.Result) {group=Erotion, parameterName=Mutiply by Base Alpha, sortPriority=3}
   Alpha=ScalarParameter() {defaultValue=1, group=Erotion, parameterName=Erosion Influence, sortPriority=-3}
  )
  FunctionOutput(None=StaticSwitchParameter(True=Saturate(None=@n13), False=Constant() {r=1}) {defaultValue=true, group=Erotion, parameterName=Erosion, sortPriority=-10}) {outputName=Color}
 ]
}
