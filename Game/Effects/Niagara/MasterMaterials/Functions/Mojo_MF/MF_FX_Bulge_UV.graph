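# MF_FX_Bulge_UV is an FX material function that applies a radial bulge distortion to UV coordinates.
# It remaps the UV input to a centered -1 to 1 range and scales it by the Bulge Intensity XY input.
# One minus the saturated squared length of that vector gives a falloff without a square root.
# The scaled vector times this falloff is subtracted from the original UV to produce Out UV.
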MaterialFunction /Game/Effects/Niagara/MasterMaterials/Functions/Mojo_MF/MF_FX_Bulge_UV
graph {nodes [
  n1 = FunctionInput(Preview=TextureCoordinate()) {bUsePreviewValueAsDefault=true, inputName=UV, inputType=FunctionInput_Vector2}
  n2 = Multiply(
   A=ConstantBiasScale(None=@n1) {bias=-0.5, desc=-.5/2, scale=2}
   B=FunctionInput(Preview=Constant2Vector() {g=0.5, r=0.5}) {bUsePreviewValueAsDefault=true, inputName=Bulge Intensity XY, inputType=FunctionInput_Vector2, sortPriority=1}
  )
  n3 = OneMinus(None=Saturate(None=DotProduct(A=@n2, B=@n2)))
  FunctionOutput(None=Subtract(A=@n1, B=Multiply(A=@n2, B=@n3))) {outputName=Out UV}
 ]}
