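# Material function that fakes an animated FFT-style noise signal for FX materials from a scrolling grayscale texture, with horizontal and vertical outputs.
# Each output scales one UV axis by a size input and places it in V with U at zero, reducing the sample to a 1D strip.
# A panner moves that strip, by default along U so values change in place, then the red channel is intensity-scaled, inverted and raised to a power.
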MaterialFunction /Game/Effects/Niagara/MasterMaterials/Functions/Mojo_MF/MF_FX_Fake_FFT_Noise
graph {nodes [
  Constant() {desc=pow, r=4}
  n1 = FunctionInput(Preview=TextureCoordinate()) {bUsePreviewValueAsDefault=true, inputName=In TexCoords, inputType=FunctionInput_Vector2}
  n2 = FunctionInput(Preview=Constant() {desc=size, r=1}) {bUsePreviewValueAsDefault=true, inputName=Fake FFT Noise Size, inputType=FunctionInput_Scalar, sortPriority=2}
  n3 = AppendVector(A=Constant(), B=Multiply(A=ComponentMask(None=@n1) {r=true}, B=@n2))
  n4 = AppendVector(
   A=FunctionInput(Preview=Constant() {desc=speed x, r=-0.139}) {bUsePreviewValueAsDefault=true, inputName=Fake FFT Noise Speed X, inputType=FunctionInput_Scalar, sortPriority=3}
   B=FunctionInput(Preview=Constant() {desc=speed y}) {bUsePreviewValueAsDefault=true, inputName=Fake FFT Noise Speed Y, inputType=FunctionInput_Scalar, sortPriority=4}
  )
  n5 = TextureObject() {samplerType=SAMPLERTYPE_LinearGrayscale, texture=/Game/Weapons/FORT_Melee/Pickaxe_Parfait_Female/FX/Textures/T_TrifleParfait_LightningNoise}
  n6 = FunctionInput(Preview=@n5) {bUsePreviewValueAsDefault=true, inputName=Fake FFT Noise Texture, inputType=FunctionInput_Texture2D, sortPriority=1}
  n7 = FunctionInput(Preview=Constant() {desc=Intensity Mult, r=0.666}) {bUsePreviewValueAsDefault=true, inputName=Fake FFT Noise Intensity Mult, inputType=FunctionInput_Scalar, sortPriority=5}
  n8 = Multiply(A=TextureSample(UVs=Panner(Coordinate=@n3, Speed=@n4), Tex=@n6).R, B=@n7)
  n9 = FunctionInput(Preview=Constant() {desc=Pow, r=2}) {bUsePreviewValueAsDefault=true, inputName=Fake FFT Noise Pow, inputType=FunctionInput_Scalar, sortPriority=6}
  FunctionOutput(None=Power(Base=OneMinus(None=@n8), Exp=@n9) {constExponent=4}) {outputName=Result Horizontal}
  n10 = AppendVector(A=Constant(), B=Multiply(A=ComponentMask(None=@n1) {g=true}, B=@n2))
  n11 = Multiply(A=TextureSample(UVs=Panner(Coordinate=@n10, Speed=@n4), Tex=@n6).R, B=@n7)
  FunctionOutput(None=Power(Base=OneMinus(None=@n11), Exp=@n9) {constExponent=4}) {outputName=Result Verticle, sortPriority=1}
 ]}
