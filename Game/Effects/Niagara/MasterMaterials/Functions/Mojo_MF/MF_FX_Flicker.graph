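# Material function for FX materials that generates a procedural, time-driven flicker value without textures.
# Scaled and offset time (Time Mult, Time Add inputs) feeds three sine waves of differing periods, with the fastest blending the other two as an unclamped lerp alpha for an irregular pulse.
# It outputs that signal as Raw Flicker, plus a biased, scaled and slightly offset copy as Processed Flicker.
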MaterialFunction /Game/Effects/Niagara/MasterMaterials/Functions/Mojo_MF/MF_FX_Flicker
graph {nodes [
  n1 = Constant() {r=1}
  n2 = Add(
   A=Multiply(A=Time(), B=FunctionInput(Preview=@n1) {bUsePreviewValueAsDefault=true, inputName=Time Mult, inputType=FunctionInput_Scalar})
   B=FunctionInput(Preview=Constant()) {bUsePreviewValueAsDefault=true, inputName=Time Add, inputType=FunctionInput_Scalar, sortPriority=1}
  )
  n3 = LinearInterpolate(A=Sine(None=@n2), B=Sine(None=@n2) {period=0.8095}, Alpha=Sine(None=@n2) {period=0.1807})
  FunctionOutput(None=Add(A=ConstantBiasScale(None=@n3) {bias=4, desc=+ *, scale=2}) {constB=0.1}) {description="Const Bias Scale (4,2)", outputName=Processed Flicker, sortPriority=1}
  FunctionOutput(None=@n3) {description=Flicker Value Unmodified, outputName=Raw Flicker}
 ]}
