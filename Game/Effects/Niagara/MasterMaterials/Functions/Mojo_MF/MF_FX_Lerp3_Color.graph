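# Material function for FX materials that blends three input colors into a gradient along a scalar t, with an adjustable middle position.
# It lerps Color 1 to Color 2 below middle and Color 2 to Color 3 above, rescaling t per segment and choosing between them with a Step.
# Alphas are not clamped, and t defaults to the U texture coordinate.
# Outputs provide RGBA and RGB results.
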MaterialFunction /Game/Effects/Niagara/MasterMaterials/Functions/Mojo_MF/MF_FX_Lerp3_Color
graph {nodes [
  Constant() {r=7.86}
  n1 = Constant4Vector() {constant=<rgba 0 0.9647 0.0529 1>}
  n2 = FunctionInput(Preview=Constant4Vector() {constant=<rgba 0.0006 0.3325 1 1>}.RGBA) {bUsePreviewValueAsDefault=true, inputName=Color 2, inputType=FunctionInput_Vector4, sortPriority=1}
  n3 = FunctionInput(Preview=Constant() {r=0.5}) {bUsePreviewValueAsDefault=true, inputName=middle, inputType=FunctionInput_Scalar, sortPriority=3}
  n4 = FunctionInput(Preview=ComponentMask(None=TextureCoordinate()) {r=true}) {bUsePreviewValueAsDefault=true, inputName=t, inputType=FunctionInput_Scalar, sortPriority=4}
  n5 = Constant4Vector() {constant=<rgba 0.7305 0.0009 1 1>}
  n6 = Subtract(A=@n4, B=@n3)
  n7 = OneMinus(None=@n3)
  n8 = LinearInterpolate(
   A=LinearInterpolate(
    A=FunctionInput(Preview=@n1.RGBA) {bUsePreviewValueAsDefault=true, inputName=Color 1, inputType=FunctionInput_Vector4}
    B=@n2
    Alpha=Divide(A=@n4, B=@n3)
   )
   B=LinearInterpolate(
    A=@n2
    B=FunctionInput(Preview=@n5.RGBA) {bUsePreviewValueAsDefault=true, inputName=Color 3, inputType=FunctionInput_Vector4, sortPriority=2}
    Alpha=Divide(A=@n6, B=@n7)
   )
   Alpha=Step(Y=@n3, X=@n4)
  )
  FunctionOutput(None=@n8) {outputName=Result V4}
  FunctionOutput(None=ComponentMask(None=@n8) {b=true, g=true, r=true}) {outputName=Result V3, sortPriority=1}
 ]}
