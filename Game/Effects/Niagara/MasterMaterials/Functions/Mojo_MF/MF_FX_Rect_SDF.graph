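# Material function that produces a rectangle SDF gradient for use in FX materials.
# It remaps input st, defaulting to UV0, from 0–1 to a centered -1–1 range, then divides each axis by the matching component of size input s.
# The maximum of the two absolute values forms the output: zero at center, one at the rectangle edge, with no thresholding or clamping.
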MaterialFunction /Game/Effects/Niagara/MasterMaterials/Functions/Mojo_MF/MF_FX_Rect_SDF
graph {
 properties {description=Produces a Rectangle SDF Gradient}
 nodes [
  n1 = TextureCoordinate()
  n2 = Subtract(
   A=Multiply(
    A=FunctionInput(Preview=@n1) {bUsePreviewValueAsDefault=true, description="st = uv", inputName=st, inputType=FunctionInput_Vector2}
    B=Constant() {r=2}
   )
   B=Constant() {r=1}
  )
  n3 = FunctionInput(Preview=Constant2Vector() {g=1, r=1}) {bUsePreviewValueAsDefault=true, inputName=s, inputType=FunctionInput_Vector2, sortPriority=1}
  n4 = Divide(A=ComponentMask(None=@n2) {r=true}, B=ComponentMask(None=@n3) {r=true})
  n5 = Divide(A=ComponentMask(None=@n2) {g=true}, B=ComponentMask(None=@n3) {g=true})
  FunctionOutput(None=Max(A=Abs(None=@n4), B=Abs(None=@n5)))
 ]
}
