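# Material function for Niagara effect master materials that builds a 2D rounded-box signed-distance mask from Position, HalfSize and CornerRadius.
# Absolute position minus HalfSize plus radius feeds outside-length and inside-min terms, then radius is subtracted.
# The inside term compares R with itself, ignoring G.
# Outputs are saturated one-minus distance and, labelled "Result One Minus", saturated distance; a time-cosine chain is unconnected.
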MaterialFunction /Game/Effects/Niagara/MasterMaterials/Functions/Mojo_MF/MF_FX_RoundedBox_SDF
graph {nodes [
  n1 = Add(A=Multiply(A=TextureCoordinate()) {constB=8}) {constB=-4}
  n2 = Constant2Vector() {g=0.4, r=0.7}
  n3 = FunctionInput(Preview=Constant() {r=1}) {inputName=CornerRadius, inputType=FunctionInput_Scalar}
  n4 = Subtract(
   A=Abs(None=FunctionInput(Preview=@n1) {bUsePreviewValueAsDefault=true, inputName=Position, inputType=FunctionInput_Vector2})
   B=Add(A=FunctionInput(Preview=@n2) {bUsePreviewValueAsDefault=true, inputName=HalfSize, inputType=FunctionInput_Vector2}, B=@n3)
  )
  n5 = MaterialFunctionCall(Float2=@n4) {materialFunction=&BreakOutFloat2Components}
  n6 = Max(A=@n5.R, B=@n5.R)
  n7 = Max(A=@n4) {constB=0}
  n8 = Subtract(A=Add(A=Length(None=@n7) {is=class MaterialExpressionLength}, B=Min(A=@n6) {constB=0}), B=@n3)
  FunctionOutput(None=Saturate(None=OneMinus(None=@n8)))
  FunctionOutput(None=Saturate(None=@n8)) {outputName=Result One Minus, sortPriority=1}
  n9 = Cosine(None=Add(A=Time()))
  Multiply(A=Multiply(A=Add(A=@n9)) {constB=0.5}) {constB=0.4}
 ]}
