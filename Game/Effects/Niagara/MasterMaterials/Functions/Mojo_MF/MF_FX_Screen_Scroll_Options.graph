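# Material function for Niagara FX materials that scrolls input UVs over time for screen-style scrolling, horizontally or vertically.
# Time scaled by a speed multiplier forms the offset, and a lerp switches between smooth motion and quantized steps built with Floor.
# The offset is added to U and V separately, and a second lerp selects the scroll axis for the output UV.
# Both switches default to stepped, horizontal scrolling.
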MaterialFunction /Game/Effects/Niagara/MasterMaterials/Functions/Mojo_MF/MF_FX_Screen_Scroll_Options
graph {nodes [
  n1 = FunctionInput(Preview=TextureCoordinate()) {bUsePreviewValueAsDefault=true, inputName=UV In, inputType=FunctionInput_Vector2}
  n2 = ComponentMask(None=@n1) {r=true}
  n3 = ComponentMask(None=@n1) {g=true}
  n4 = Multiply(A=Time(), B=FunctionInput(Preview=Constant() {r=1}) {bUsePreviewValueAsDefault=true, inputName=Speed Mult, inputType=FunctionInput_Scalar})
  n5 = Add(A=Multiply(A=@n4) {constB=8}) {constB=0.5}
  n6 = LinearInterpolate(
   A=@n4
   B=Multiply(A=Floor(None=@n5)) {constB=0.1}
   Alpha=FunctionInput(Preview=Constant() {r=1}) {bUsePreviewValueAsDefault=true, inputName=Stepped Scroll On/Off, inputType=FunctionInput_Scalar}
  )
  n7 = Add(A=@n3, B=@n6)
  n8 = Add(A=@n2, B=@n6)
  n9 = FunctionInput(Preview=Constant() {r=1}) {
   bUsePreviewValueAsDefault=true
   description="0 = Vert, 1 = Horizontal"
   inputName=Vertical / Horizontal Scroll
   inputType=FunctionInput_Scalar
  }
  FunctionOutput(None=LinearInterpolate(A=AppendVector(A=@n2, B=@n7) {desc=Vert Scroll}, B=AppendVector(A=@n8, B=@n3) {desc=Horizontal Scroll}, Alpha=@n9)) {outputName=UV Out}
 ]}
