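# Material function for FX materials that builds a square border mask along the four edges of the input UVs, with an inverted version.
# Flooring UV plus width and one-minus UV plus width flags each edge; the channels are summed and saturated into a hard-edged mask.
# A second output pair derives per-axis widths from Width and ObjectScale X and Y, with both axes measuring Scale X's distance from Scale Y minus Width.
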MaterialFunction /Game/Effects/Niagara/MasterMaterials/Functions/Mojo_MF/MF_FX_SquareEdgeMask
graph {nodes [
  n1 = FunctionInput(Preview=TextureCoordinate()) {bUsePreviewValueAsDefault=true, inputName=In TexCoords, inputType=FunctionInput_Vector2}
  n2 = MaterialFunctionCall() {materialFunction=&ObjectScale}
  n3 = FunctionInput(Preview=Constant() {r=0.075}) {bUsePreviewValueAsDefault=true, inputName=Width, inputType=FunctionInput_Scalar, sortPriority=1}
  n4 = Subtract(A=@n2.Scale Y, B=@n3)
  n5 = Subtract(A=@n2.Scale Y, B=@n3)
  n6 = AppendVector(A=Divide(A=Distance(A=@n2.Scale X, B=@n4), B=@n2.Scale X), B=Divide(A=Distance(A=@n2.Scale X, B=@n5), B=@n2.Scale Y))
  n7 = Floor(None=Add(A=OneMinus(None=@n1), B=@n6))
  n8 = ComponentMask(None=@n7) {r=true}
  n9 = ComponentMask(None=@n7) {g=true}
  n10 = Floor(None=Add(A=@n1, B=@n6))
  n11 = ComponentMask(None=@n10) {r=true}
  n12 = ComponentMask(None=@n10) {g=true}
  n13 = Saturate(None=Add(A=Add(A=@n8, B=@n9), B=Add(A=@n11, B=@n12)))
  FunctionOutput(None=@n13) {outputName="Mask (Object Scale)"}
  FunctionOutput(None=OneMinus(None=@n13)) {outputName="Mask One Minus (Object Scale)", sortPriority=1}
  n14 = Floor(None=Add(A=OneMinus(None=@n1), B=@n3))
  n15 = ComponentMask(None=@n14) {r=true}
  n16 = ComponentMask(None=@n14) {g=true}
  n17 = Floor(None=Add(A=@n1, B=@n3))
  n18 = ComponentMask(None=@n17) {r=true}
  n19 = ComponentMask(None=@n17) {g=true}
  n20 = Saturate(None=Add(A=Add(A=@n15, B=@n16), B=Add(A=@n18, B=@n19)))
  FunctionOutput(None=@n20) {outputName=Mask, sortPriority=2}
  FunctionOutput(None=OneMinus(None=@n20)) {outputName=Mask One Minus, sortPriority=3}
 ]}
