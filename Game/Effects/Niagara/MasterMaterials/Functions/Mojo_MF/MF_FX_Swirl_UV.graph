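# Material function that produces swirled, time-animated UVs for FX materials, twisting coordinates around the texture center.
# Radial distance from UV 0.5 becomes a saturated linear falloff, scaled by Spin, so the twist is strongest at the center and zero at radius 0.5.
# Time times Speed times two Pi adds uniform rotation.
# The summed angle drives CustomRotator on the input UVs, despite that node expecting a 0-1 angle.
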MaterialFunction /Game/Effects/Niagara/MasterMaterials/Functions/Mojo_MF/MF_FX_Swirl_UV
graph {nodes [
  n1 = Constant() {r=0.5}
  n2 = FunctionInput(Preview=TextureCoordinate()) {bUsePreviewValueAsDefault=true, inputName=UV TexCoord, inputType=FunctionInput_Vector2}
  n3 = Subtract(A=@n2, B=@n1)
  n4 = MaterialFunctionCall(Vector 2=Abs(None=Divide(A=@n3, B=@n1))) {materialFunction=/Engine/Functions/Engine_MaterialFunctions02/Utility/VectorLength}
  n5 = Multiply(
   A=Saturate(None=OneMinus(None=@n4.V2 Length))
   B=FunctionInput(Preview=Constant() {r=1}) {bUsePreviewValueAsDefault=true, inputName=Spin, inputType=FunctionInput_Scalar, sortPriority=3}
  )
  n6 = Time()
  n7 = Constant() {r=0.001}
  n8 = Multiply(
   A=Multiply(
    A=FunctionInput(Preview=@n6) {bUsePreviewValueAsDefault=true, inputName=Time, inputType=FunctionInput_Scalar, sortPriority=1}
    B=FunctionInput(Preview=@n7) {bUsePreviewValueAsDefault=true, inputName=Speed, inputType=FunctionInput_Scalar, sortPriority=2}
   )
   B=MaterialFunctionCall(Multiplier=Constant() {r=2}) {materialFunction=&Pi}.Result
  )
  FunctionOutput(None=MaterialFunctionCall(UVs=@n2, "Rotation Angle (0-1)"=Add(A=@n5, B=@n8)) {materialFunction=&CustomRotator}.Rotated Values) {outputName=Out UV TexCoord}
 ]}
