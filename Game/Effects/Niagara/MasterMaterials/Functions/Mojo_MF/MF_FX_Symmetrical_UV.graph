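# Material function for effects materials that remaps UVs so they mirror symmetrically about the center of one axis.
# It splits the input UV, defaulting to TexCoord 0, and turns a channel into 1 - 4x(1-x), a parabola that is 0 at center and 1 at edges.
# "Symmetric UV Horiz" keeps U and remaps V, while "Symmetric UV Vert" remaps U and keeps V.
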MaterialFunction /Game/Effects/Niagara/MasterMaterials/Functions/Mojo_MF/MF_FX_Symmetrical_UV
graph {nodes [
  n1 = MaterialFunctionCall(Float2=FunctionInput(Preview=TextureCoordinate()) {bUsePreviewValueAsDefault=true, inputName=In UV, inputType=FunctionInput_Vector2}) {materialFunction=&BreakOutFloat2Components}
  n2 = Multiply(A=Multiply(A=@n1.G, B=OneMinus(None=@n1.G)), B=Constant() {r=4})
  FunctionOutput(None=AppendVector(A=@n1.R, B=OneMinus(None=@n2))) {outputName=Symmetric UV Horiz}
  n3 = Multiply(A=Multiply(A=@n1.R, B=OneMinus(None=@n1.R)), B=Constant() {r=4})
  FunctionOutput(None=AppendVector(A=OneMinus(None=@n3), B=@n1.G)) {outputName=Symmetric UV Vert}
 ]}
