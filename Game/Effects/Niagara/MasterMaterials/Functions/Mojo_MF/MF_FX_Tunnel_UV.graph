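# Niagara FX material function that remaps texture coordinates into scrolling polar UVs for an infinite-tunnel effect.
# Centered UVs give an arctangent angle for U and Size divided by radius for V, each offset by time and its own speed.
# It outputs the tunnel UVs, depth term, saturated radius mask and centered UVs.
# An unconnected debug preview samples with centered-UV derivatives, sidestepping the angular seam in mip selection.
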MaterialFunction /Game/Effects/Niagara/MasterMaterials/Functions/Mojo_MF/MF_FX_Tunnel_UV
graph {nodes [
  n1 = TextureCoordinate()
  n2 = Add(
   A=Constant() {r=-1}
   B=Multiply(A=FunctionInput(Preview=@n1) {bUsePreviewValueAsDefault=true, inputName=In TexCoords, inputType=FunctionInput_Vector2}, B=Constant() {r=2})
  ) {desc=p}
  n3 = Arctangent2!(Y=ComponentMask(None=@n2) {g=true}, X=ComponentMask(None=@n2) {r=true})
  n4 = FunctionInput(Preview=Constant() {desc=Pi, r=3.1416}) {bUsePreviewValueAsDefault=true, inputName=Divide Control, inputType=FunctionInput_Scalar, sortPriority=5}
  n5 = FunctionInput(Preview=Constant() {desc=speed X, r=0.1}) {bUsePreviewValueAsDefault=true, inputName=Speed X, inputType=FunctionInput_Scalar, sortPriority=2}
  n6 = FunctionInput(Preview=Time()) {bUsePreviewValueAsDefault=true, inputName=Time, inputType=FunctionInput_Scalar, sortPriority=1}
  n7 = FunctionInput(Preview=Constant() {desc=speed Y, r=0.25}) {bUsePreviewValueAsDefault=true, inputName=Speed Y, inputType=FunctionInput_Scalar, sortPriority=3}
  n8 = SquareRoot(None=DotProduct(A=@n2, B=@n2)) {desc=r}
  n9 = Divide(
   A=FunctionInput(Preview=Constant() {desc=size, r=0.25}) {bUsePreviewValueAsDefault=true, inputName=Size, inputType=FunctionInput_Scalar, sortPriority=4}
   B=@n8
  )
  n10 = AppendVector(A=Add(A=Divide(A=@n3, B=@n4), B=Multiply(A=@n6, B=@n5) {desc=time}), B=Add(A=Multiply(A=@n6, B=@n7) {desc=time}, B=@n9))
  FunctionOutput(None=@n10) {outputName=Tunnel UV Static, sortPriority=1}
  FunctionOutput(None=@n9) {outputName=Tunnel Size Mask, sortPriority=2}
  n11 = Saturate(None=@n8)
  FunctionOutput(None=@n11) {outputName=Tunnel Saturated Mask, sortPriority=3}
  FunctionOutput(None=@n2) {outputName=uv}
  Multiply(
   A=TextureSample!(
    UVs=@n10
    Tex=TextureObject() {texture=/Game/Effects/Fort_Effects/Textures/T_FX_Circuitboard}
    "DDX(UVs)"=DDX(Value=@n2)
    "DDY(UVs)"=DDY(Value=@n2)
    Apply View MipBias
   ) {mipValueMode=TMVM_Derivative, texture=/Game/Effects/Fort_Effects/Effects/Characters/Generic/ResIn/DiffusedStylizedGrid}.RGB
   B=@n11
  ) {desc=Debug Preview}
  TextureCoordinate()
 ]}
