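# Material function that distorts UVs with noise offsets rotating over time, for animated swirling distortion in effect materials.
# Red and blue channels of a tiled noise texture form centered offset vectors, which a Custom HLSL node rotates through full turns using wrapped Time scaled by Turn Time.
# The scaled offsets shift the tiled UVs, yielding the sampled input texture, base UVs and detail UVs with tiling and distortion raised by Detail Factor.
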MaterialFunction /Game/Effects/Niagara/MasterMaterials/Functions/Mojo_MF/MF_FX_TurnDistortion
graph {nodes [
  n1 = TextureObject() {samplerType=SAMPLERTYPE_LinearColor, texture=/Game/Global/Textures/Default/Debugs/CenterpointDebugTexture}
  n2 = FunctionInput(Preview=TextureCoordinate()) {bUsePreviewValueAsDefault=true, inputName=In Texcoords, inputType=FunctionInput_Vector2, sortPriority=2}
  n3 = Constant2Vector() {desc=Distortion Tiling, g=1, r=1}
  n4 = TextureSample(
   UVs=Multiply(
    A=@n2
    B=FunctionInput(Preview=@n3) {bUsePreviewValueAsDefault=true, inputName=Distortion Tiling, inputType=FunctionInput_Vector2, sortPriority=4}
   )
   Tex=FunctionInput(Preview=TextureObject() {texture=/Game/Global/Textures/Default/Noise/T_Noise_Cloud}) {bUsePreviewValueAsDefault=true, inputName=Random Color Texture, inputType=FunctionInput_Texture2D, sortPriority=1}
  )
  n5 = ComponentMask(None=@n4.RGB) {r=true}
  n6 = ComponentMask(None=@n4.RGB) {b=true}
  MaterialExpressionReroute_52 = Reroute(None=Subtract(A=Constant() {r=0.5}, B=AppendVector(A=@n5, B=@n6))) {desc=dir}
  n7 = Time()
  n8 = Divide(B=FunctionInput(Preview=Constant() {desc=Turn Time, r=0.25}) {bUsePreviewValueAsDefault=true, inputName=Turn Time, inputType=FunctionInput_Scalar, sortPriority=3})
  MaterialExpressionReroute_53 = Reroute(None=Fmod(A=Divide(A=@n7, B=@n8), B=Constant() {r=1})) {desc=turn}
  n9 = Custom!(dir=@MaterialExpressionReroute_52, turn=@MaterialExpressionReroute_53) {
   code="float dist = length(dir);\r\nfloat angle = 0;\r\nif (dir.x!=0) angle = atan2(dir.x, dir.y);\r\nangle += turn * 6.28318530717958647693;//2Pi\r\ndir = float2(sin(angle), cos(angle));\r\ndir*=dist;\r\n\r\nreturn dir;"
   description=turn should be in the 0-1 range
   outputType=CMOT_Float2
   inputs [
    {inputName=dir, input {expression=&self:MaterialExpressionReroute_52}}
    {inputName=turn, input {expression=&self:MaterialExpressionReroute_53}}
   ]
  }
  n10 = Multiply(
   A=@n9
   B=FunctionInput(Preview=Constant() {desc=Distortion, r=0.05}) {bUsePreviewValueAsDefault=true, inputName=Distortion Amount, inputType=FunctionInput_Scalar, sortPriority=5}
  )
  n11 = FunctionInput(Preview=Constant2Vector() {desc=Tiling, g=1, r=1}) {bUsePreviewValueAsDefault=true, inputName=Tiling, inputType=FunctionInput_Vector2, sortPriority=6}
  n12 = Add(A=Multiply(A=@n2, B=@n11), B=@n10) {desc=distort base}
  FunctionOutput(None=TextureSample(UVs=@n12, Tex=FunctionInput(Preview=@n1) {bUsePreviewValueAsDefault=true, inputName=In Texture, inputType=FunctionInput_Texture2D}).RGBA)
  FunctionOutput(None=@n12) {outputName=Out Base UV, sortPriority=1}
  n13 = FunctionInput(Preview=Constant() {desc=Detail Tile Factor, r=5}) {bUsePreviewValueAsDefault=true, inputName=Detail Factor, inputType=FunctionInput_Scalar, sortPriority=7}
  n14 = Multiply(A=@n13) {constB=0.8, desc=detail distortion factor}
  n15 = Multiply(A=@n11, B=@n13)
  FunctionOutput(None=Add(A=Multiply(A=@n2, B=@n15), B=Multiply(A=@n10, B=@n14)) {desc=Distort Detail}) {outputName=Out Detail UV, sortPriority=2}
 ]}
