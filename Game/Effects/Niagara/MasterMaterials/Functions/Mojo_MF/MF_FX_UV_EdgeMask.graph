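# Material function for FX materials that builds a procedural, texture-free mask highlighting the edges of UV space.
# U and V each pass through an inverted half-period sine, peaking at the 0 and 1 borders, then get raised to Edge Sharpness and offset by Edge Subtract.
# A default-alpha lerp averages both axes without clamping, outputting Edge Mask and its one-minus inverse.
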MaterialFunction /Game/Effects/Niagara/MasterMaterials/Functions/Mojo_MF/MF_FX_UV_EdgeMask
graph {nodes [
  n1 = FunctionInput(Preview=TextureCoordinate()) {bUsePreviewValueAsDefault=true, inputName=TexCoords, inputType=FunctionInput_Vector2}
  n2 = ComponentMask(None=@n1) {r=true}
  n3 = OneMinus(None=Sine(None=Multiply(A=@n2) {constB=0.5}))
  n4 = FunctionInput(Preview=Constant() {r=30}) {bUsePreviewValueAsDefault=true, inputName=Edge Sharpness, inputType=FunctionInput_Scalar, sortPriority=1}
  n5 = FunctionInput(Preview=Constant() {r=0.1}) {bUsePreviewValueAsDefault=true, inputName=Edge Subtract, inputType=FunctionInput_Scalar, sortPriority=2}
  n6 = ComponentMask(None=@n1) {g=true}
  n7 = OneMinus(None=Sine(None=Multiply(A=@n6) {constB=0.5}))
  n8 = LinearInterpolate(A=Subtract(A=Power(Base=@n3, Exp=@n4), B=@n5), B=Subtract(A=Power(Base=@n7, Exp=@n4), B=@n5))
  FunctionOutput(None=@n8) {outputName=Edge Mask}
  FunctionOutput(None=OneMinus(None=@n8)) {outputName=1-Edge Mask, sortPriority=1}
 ]}
